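# Niagara system for the PopDroid pickaxe's swing trail, built from three ribbon emitters that use shared AnimTrail modules on bone or socket indices.
# Trail_Dark adds a short, translucent black ribbon, Trail_Body a two-renderer colored ribbon with a secondary color, and Trail_Main a thin HDR streak.
# Body and Main fade over system age using curves selected by User.SwingSectionIndex.
# All three also scale alpha by a time-of-day value.
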NiagaraSystem /BRCosmetics/Weapons/FORT_Melee/Pickaxe_PopDroid/FX/NS_Pickaxe_PopDroid_Trail
dependencies {
 usedDataInterfaces [&NiagaraDataInterfaceCurve, &NiagaraDataInterfaceColorCurve, &NiagaraDataInterfaceSocketReader]
 usedDynamicInputs [&FNTimeOfDayScalar]
 usedModules [
  &self:SystemSpawnScript.SystemScriptSource.SystemScriptGraph.NiagaraNodeAssignment_5.SetVariables_#1
  &SystemState
  &self:SystemSpawnScript.SystemScriptSource.SystemScriptGraph.NiagaraNodeAssignment_3.SetVariables_#2
  &self:SystemSpawnScript.SystemScriptSource.SystemScriptGraph.NiagaraNodeAssignment_1.SetVariables_#3
  &self:SystemSpawnScript.SystemScriptSource.SystemScriptGraph.NiagaraNodeAssignment_2.SetVariables_#4
 ]
}
emitters {
 Trail_Body {
  emitterUpdateScript [
   &EmitterState()
   &FN_SpawnAnimTrail()
   /SetVariables(Emitter.rand:NiagaraFloat=&RandomRangeFloat(Minimum=1, Maximum=1, Evaluation Type=::NewEnumerator1)) {isSetParameters=true}
  ]
  particleSpawnScript [
   &FN_InitializeAnimTrailBoneIndecies(First Bone or Socket Index=1, Second Bone or Socket Index=0)
   &FN_InitializeAnimTrail(AnimTrailWidth=1.2, Color=<rgba 2.474 4.5925 5 1>, Lifetime=0.27)
   /SetVariables(Particles.Position_Tip:NiagaraPosition=&AddVectorToPosition(Position=$Particles.Position, Vector=&TransformVector(Vector=<0 0 -20>, Destination Space=::Simulation))) {isSetParameters=true}
  ]
  particleUpdateScript [
   &FN_UpdateAnimTrail()
   &ParticleState()
   &ScaleColor(
    Scale Alpha=&Multiply_Float(
     A=&MergeHelper_FloatSwitch(
      0=&FloatFromCurve(
       FloatCurve=DataInterface() {propertyValues {Curve {
          defaultValue=1
          keys [
           {arriveTangent=0.074, interpMode=RCIM_Cubic, leaveTangent=0.074, tangentMode=RCTM_User, time=0, value=1}
           {arriveTangent=-8.6402, interpMode=RCIM_Cubic, leaveTangent=-8.6402, tangentMode=RCTM_User, time=0.1972, value=0.5421}
           {arriveTangent=0.0063, interpMode=RCIM_Cubic, leaveTangent=0.0063, tangentMode=RCTM_User, time=0.26, value=0}
          ]
         }}}
       CurveIndex=$Engine.System.Age
      )
      1=&FloatFromCurve(
       FloatCurve=DataInterface() {propertyValues {Curve {
          defaultValue=1
          keys [
           {arriveTangent=0.074, interpMode=RCIM_Cubic, leaveTangent=0.074, tangentMode=RCTM_User, time=0, value=1}
           {arriveTangent=-8.6402, interpMode=RCIM_Cubic, leaveTangent=-8.6402, tangentMode=RCTM_User, time=0.1915, value=0.6}
           {arriveTangent=0.0063, interpMode=RCIM_Cubic, leaveTangent=0.0063, tangentMode=RCTM_User, time=0.25, value=0}
          ]
         }}}
       CurveIndex=$Engine.System.Age
      )
      2=&FloatFromCurve(
       FloatCurve=DataInterface() {propertyValues {Curve {
          defaultValue=1
          keys [
           {arriveTangent=0.074, interpMode=RCIM_Cubic, leaveTangent=0.074, tangentMode=RCTM_User, time=0, value=1}
           {arriveTangent=-8.6402, interpMode=RCIM_Cubic, leaveTangent=-8.6402, tangentMode=RCTM_User, time=0.1779, value=0.5777}
           {arriveTangent=0.0063, interpMode=RCIM_Cubic, leaveTangent=0.0063, tangentMode=RCTM_User, time=0.25, value=0}
          ]
         }}}
       CurveIndex=$Engine.System.Age
      )
      3=&FloatFromCurve(
       FloatCurve=DataInterface() {propertyValues {Curve {
          defaultValue=1
          keys [
           {arriveTangent=0.074, interpMode=RCIM_Cubic, leaveTangent=0.074, tangentMode=RCTM_User, time=0, value=1}
           {arriveTangent=-11.458, interpMode=RCIM_Cubic, leaveTangent=-11.458, tangentMode=RCTM_User, time=0.1476, value=0.4947}
           {arriveTangent=0.0063, interpMode=RCIM_Cubic, leaveTangent=0.0063, tangentMode=RCTM_User, time=0.2, value=0}
          ]
         }}}
       CurveIndex=$Engine.System.Age
      )
      Index=$User.SwingSectionIndex
      Num=::NewEnumerator2
     )
     B=&FloatFromCurve(
      FloatCurve=DataInterface() {propertyValues {Curve {keys [{interpMode=RCIM_Cubic, tangentMode=RCTM_User, time=0, value=1}, {interpMode=RCIM_Cubic, tangentMode=RCTM_User, time=1, value=0}]}}}
      Scale Curve=$System.TOD
     )
    )
    Scale RGB=&MakeVectorFromLinearColorRGB(Color=&ColorFromCurve(ColorCurve=DataInterface() {propertyValues {
        AlphaCurve {keys [<at 0 1 RCIM_Cubic>,<at 1 0 RCIM_Cubic>]}
        BlueCurve {keys [<at 0 0.18 RCIM_Cubic>]}
        GreenCurve {keys [<at 0 0.18 RCIM_Cubic>]}
        RedCurve {keys [<at 0 0.18 RCIM_Cubic>]}
       }}))
    ScaleA=true
    ScaleRGB=true
   )
   &DynamicMaterialParameters(
    Index 0 Param 2=&FloatFromCurve(FloatCurve=DataInterface() {propertyValues {Curve {
        defaultValue=0
        keys [{interpMode=RCIM_Cubic, tangentMode=RCTM_User, time=0, value=1}, {interpMode=RCIM_Cubic, tangentMode=RCTM_User, time=1, value=0}]
       }}})
    Module.Param0WriteEnabled=false
    Module.Param1WriteEnabled=true
    Module.Param2WriteEnabled=false
    Module.Param3WriteEnabled=false
   )
   /SetVariables(Particles.Trail Color02:LinearColor=&MakeLinearColorFromVectorAndFloat(
     "Vector (RGB)"=&Multiply_VectorByFloat(
      Vector=&MakeVectorFromLinearColorRGB(Color=&ColorFromCurve(ColorCurve=DataInterface() {propertyValues {
          AlphaCurve {keys [<at 0 1 RCIM_Cubic>,<at 1 0 RCIM_Cubic>]}
          BlueCurve {keys [<at 0 0.4812 RCIM_Cubic>,<at 1 0.5469 RCIM_Cubic>]}
          GreenCurve {keys [<at 0 0.1492 RCIM_Cubic>,<at 1 0.0217 RCIM_Cubic>]}
          RedCurve {keys [<at 0 0.5729 RCIM_Cubic>,<at 1 0.113 RCIM_Cubic>]}
         }}))
      Float=1
     )
     "Float (Alpha)"=&Multiply_Float(
      A=&MergeHelper_FloatSwitch(
       0=&FloatFromCurve(
        FloatCurve=DataInterface() {propertyValues {Curve {
           defaultValue=1
           keys [
            {arriveTangent=0.074, interpMode=RCIM_Cubic, leaveTangent=0.074, tangentMode=RCTM_User, time=0, value=1}
            {arriveTangent=-8.6402, interpMode=RCIM_Cubic, leaveTangent=-8.6402, tangentMode=RCTM_User, time=0.1972, value=0.5421}
            {arriveTangent=0.0063, interpMode=RCIM_Cubic, leaveTangent=0.0063, tangentMode=RCTM_User, time=0.26, value=0}
           ]
          }}}
        CurveIndex=$Engine.System.Age
       )
       1=&FloatFromCurve(
        FloatCurve=DataInterface() {propertyValues {Curve {
           defaultValue=1
           keys [
            {arriveTangent=0.074, interpMode=RCIM_Cubic, leaveTangent=0.074, tangentMode=RCTM_User, time=0, value=1}
            {arriveTangent=-8.6402, interpMode=RCIM_Cubic, leaveTangent=-8.6402, tangentMode=RCTM_User, time=0.1915, value=0.6}
            {arriveTangent=0.0063, interpMode=RCIM_Cubic, leaveTangent=0.0063, tangentMode=RCTM_User, time=0.25, value=0}
           ]
          }}}
        CurveIndex=$Engine.System.Age
       )
       2=&FloatFromCurve(
        FloatCurve=DataInterface() {propertyValues {Curve {
           defaultValue=1
           keys [
            {arriveTangent=0.074, interpMode=RCIM_Cubic, leaveTangent=0.074, tangentMode=RCTM_User, time=0, value=1}
            {arriveTangent=-8.6402, interpMode=RCIM_Cubic, leaveTangent=-8.6402, tangentMode=RCTM_User, time=0.1779, value=0.5777}
            {arriveTangent=0.0063, interpMode=RCIM_Cubic, leaveTangent=0.0063, tangentMode=RCTM_User, time=0.25, value=0}
           ]
          }}}
        CurveIndex=$Engine.System.Age
       )
       3=&FloatFromCurve(
        FloatCurve=DataInterface() {propertyValues {Curve {
           defaultValue=1
           keys [
            {arriveTangent=0.074, interpMode=RCIM_Cubic, leaveTangent=0.074, tangentMode=RCTM_User, time=0, value=1}
            {arriveTangent=-11.458, interpMode=RCIM_Cubic, leaveTangent=-11.458, tangentMode=RCTM_User, time=0.1476, value=0.4947}
            {arriveTangent=0.0063, interpMode=RCIM_Cubic, leaveTangent=0.0063, tangentMode=RCTM_User, time=0.2, value=0}
           ]
          }}}
        CurveIndex=$Engine.System.Age
       )
       Index=$User.SwingSectionIndex
       Num=::NewEnumerator2
      )
      B=&FloatFromCurve(
       FloatCurve=DataInterface() {propertyValues {Curve {keys [
           {interpMode=RCIM_Cubic, tangentMode=RCTM_User, time=0, value=1}
           {arriveTangent=-1.4244, interpMode=RCIM_Cubic, leaveTangent=-1.4244, tangentMode=RCTM_User, time=1, value=0}
          ]}}}
       Scale Curve=$System.TOD
      )
     )
    )) {isSetParameters=true}
   /SetVariables(Particles.DY2:Vector4f=&MakeVector4(X=1, Y=1, Z=1, W=1)) {isSetParameters=true}
  ]
  rendererClasses [&NiagaraRibbonRendererProperties]
  renderers [
   {rendererClass=&NiagaraRibbonRendererProperties}
   {rendererClass=&NiagaraRibbonRendererProperties, rendererIndex=1}
  ]
 }
 Trail_Dark {
  emitterUpdateScript [
   &EmitterState()
   &FN_SpawnAnimTrail_v2()
  ]
  particleSpawnScript [
   /Game/Effects/Niagara/Modules/AnimTrail/FN_InitializeAnimTrail/FN_InitializeAnimTrailBoneIndecies(Second Bone or Socket Index=1)
   &FN_InitializeAnimTrail(AnimTrailWidth=1.5, Color=<rgba 0 0 0 0.45>, Lifetime=0.15) {module=FN_InitializeAnimTrail001}
   /SetVariables(Particles.Position_Tip:NiagaraPosition=&AddVectorToPosition(Position=$Particles.Position, Vector=&TransformVector(Vector=<0 0 -15>, Destination Space=::Simulation))) {isSetParameters=true}
  ]
  particleUpdateScript [
   &FN_UpdateAnimTrail()
   &ScaleRibbonWidth(Ribbon Width Scale=&FloatFromCurve(FloatCurve=DataInterface() {propertyValues {Curve {keys [
         {interpMode=RCIM_Cubic, tangentMode=RCTM_User, time=0, value=1.1}
         {arriveTangent=-2.3701, interpMode=RCIM_Cubic, leaveTangent=-2.3701, tangentMode=RCTM_User, time=1, value=0}
        ]}}}))
   &SpriteFacingAndAlignment(Sprite Facing=$Particles.RibbonFacing, WriteAlignment=false, WriteFacing=true)
   &DynamicMaterialParameters(
    Index 0 Param 1=&FloatFromCurve(FloatCurve=DataInterface() {propertyValues {Curve {keys [<at 0 1 RCIM_Cubic>, {arriveTangent=-1, interpMode=RCIM_Cubic, leaveTangent=-1, time=0.5, value=0.8}, <at 1 0>]}}})
    Module.Param0WriteEnabled=true
    Module.Param1WriteEnabled=true
    Module.Param2WriteEnabled=true
    Module.Param3WriteEnabled=true
   )
   &ScaleColor(
    Scale Alpha=&FloatFromCurve(
     FloatCurve=DataInterface() {propertyValues {Curve {
        defaultValue=1
        keys [{interpMode=RCIM_Cubic, tangentMode=RCTM_User, time=0, value=1}, {interpMode=RCIM_Cubic, tangentMode=RCTM_User, time=1, value=0}]
       }}}
     Scale Curve=$System.TOD
    )
    Scale RGB=&MakeVectorFromLinearColorRGB(Color=$System.TODColor)
    ScaleA=true
    ScaleRGB=true
   )
  ]
  rendererClasses [&NiagaraRibbonRendererProperties]
  renderers [{rendererClass=&NiagaraRibbonRendererProperties}]
 }
 Trail_Main {
  emitterUpdateScript [
   &EmitterState()
   &FN_SpawnAnimTrail()
   /SetVariables(Emitter.rand:NiagaraFloat=&RandomRangeFloat(Minimum=1, Maximum=1, Evaluation Type=::NewEnumerator1)) {isSetParameters=true}
  ]
  particleSpawnScript [
   /Game/Effects/Niagara/Modules/AnimTrail/FN_InitializeAnimTrailBoneIndecies_v2/FN_InitializeAnimTrailBoneIndecies()
   &FN_InitializeAnimTrail(AnimTrailWidth=0.2, Lifetime=0.33)
   /SetVariables(Particles.Position_Tip:NiagaraPosition=&AddVectorToPosition(
     Position=$Particles.Position
     Vector=&TransformVector(
      Vector=&MergeHelper_VectorSwitch(0=<0 0 0>, 1=<0 0 0>, 2=<0 0 0>, 3=<0 0 0>, Index=$User.SwingSectionIndex, Num=::NewEnumerator2)
      Destination Space=::Simulation
     )
    )) {isSetParameters=true}
  ]
  particleUpdateScript [
   &ParticleState()
   &FN_UpdateAnimTrail()
   &ScaleColor(
    Scale Alpha=&Multiply_Float(
     A=&MergeHelper_FloatSwitch(
      0=&FloatFromCurve(
       FloatCurve=DataInterface() {propertyValues {Curve {
          defaultValue=1
          keys [
           {arriveTangent=0.074, interpMode=RCIM_Cubic, leaveTangent=0.074, tangentMode=RCTM_User, time=0, value=1}
           {arriveTangent=-8.6402, interpMode=RCIM_Cubic, leaveTangent=-8.6402, tangentMode=RCTM_User, time=0.1972, value=0.5421}
           {arriveTangent=0.0063, interpMode=RCIM_Cubic, leaveTangent=0.0063, tangentMode=RCTM_User, time=0.26, value=0}
          ]
         }}}
       CurveIndex=$Engine.System.Age
      )
      1=&FloatFromCurve(
       FloatCurve=DataInterface() {propertyValues {Curve {
          defaultValue=1
          keys [
           {arriveTangent=0.074, interpMode=RCIM_Cubic, leaveTangent=0.074, tangentMode=RCTM_User, time=0, value=1}
           {arriveTangent=-8.6402, interpMode=RCIM_Cubic, leaveTangent=-8.6402, tangentMode=RCTM_User, time=0.1915, value=0.6}
           {arriveTangent=0.0063, interpMode=RCIM_Cubic, leaveTangent=0.0063, tangentMode=RCTM_User, time=0.25, value=0}
          ]
         }}}
       CurveIndex=$Engine.System.Age
      )
      2=&FloatFromCurve(
       FloatCurve=DataInterface() {propertyValues {Curve {
          defaultValue=1
          keys [
           {arriveTangent=0.074, interpMode=RCIM_Cubic, leaveTangent=0.074, tangentMode=RCTM_User, time=0, value=1}
           {arriveTangent=-8.6402, interpMode=RCIM_Cubic, leaveTangent=-8.6402, tangentMode=RCTM_User, time=0.1779, value=0.5777}
           {arriveTangent=0.0063, interpMode=RCIM_Cubic, leaveTangent=0.0063, tangentMode=RCTM_User, time=0.25, value=0}
          ]
         }}}
       CurveIndex=$Engine.System.Age
      )
      3=&FloatFromCurve(
       FloatCurve=DataInterface() {propertyValues {Curve {
          defaultValue=1
          keys [
           {arriveTangent=0.074, interpMode=RCIM_Cubic, leaveTangent=0.074, tangentMode=RCTM_User, time=0, value=1}
           {arriveTangent=-11.458, interpMode=RCIM_Cubic, leaveTangent=-11.458, tangentMode=RCTM_User, time=0.1476, value=0.4947}
           {arriveTangent=0.0063, interpMode=RCIM_Cubic, leaveTangent=0.0063, tangentMode=RCTM_User, time=0.2, value=0}
          ]
         }}}
       CurveIndex=$Engine.System.Age
      )
      Index=$User.SwingSectionIndex
      Num=::NewEnumerator2
     )
     B=&FloatFromCurve(
      FloatCurve=DataInterface() {propertyValues {Curve {keys [{interpMode=RCIM_Cubic, tangentMode=RCTM_User, time=0, value=1}, {interpMode=RCIM_Cubic, tangentMode=RCTM_User, time=1, value=0}]}}}
      Scale Curve=$System.TOD
     )
    )
    Scale RGB=&Multiply_VectorByFloat(
     Vector=&MakeVectorFromLinearColorRGB(Color=&ColorFromCurve(ColorCurve=DataInterface() {propertyValues {
         AlphaCurve {keys [<at 0 1 RCIM_Cubic>,<at 1 0 RCIM_Cubic>]}
         BlueCurve {keys [<at 0 13.0636 RCIM_Cubic>,<at 0.3629 2 RCIM_Cubic>]}
         GreenCurve {keys [<at 0 10.1772 RCIM_Cubic>,<at 0.3629 0 RCIM_Cubic>]}
         RedCurve {keys [<at 0 15 RCIM_Cubic>,<at 0.3629 1.8087 RCIM_Cubic>]}
        }}))
     Float=&MakeFloatFromLinearColor(LinearColor=$System.TODColor, Channel:ENiagaraFloatFromLinearColorOptions=::NewEnumerator0)
    )
    ScaleA=true
    ScaleRGB=true
   )
   &DynamicMaterialParameters(
    Index 0 Param 1=2.5
    Module.Index0Param0Write=true
    Module.Index0Param1Write=true
    Module.Index0Param2Write=true
    Module.Index0Param3Write=true
   )
  ]
  rendererClasses [&NiagaraRibbonRendererProperties]
  renderers [{rendererClass=&NiagaraRibbonRendererProperties}]
 }
}
summary {
 systemName=NS_Pickaxe_PopDroid_Trail
 emitters [
  {emitterName=Trail_Dark, rendererClasses [&NiagaraRibbonRendererProperties]}
  {emitterName=Trail_Body, rendererClasses [&NiagaraRibbonRendererProperties]}
  {emitterName=Trail_Main, rendererClasses [&NiagaraRibbonRendererProperties]}
 ]
 userVariables [
  User.bIsFrontEndPreview {type=NiagaraBool}
  User.bIsOffhand {type=NiagaraBool}
  User.bIsRightHand {type=NiagaraBool}
  User.Color A {type=LinearColor, description="", defaultValue {struct=LinearColor, value=<rgba 1 1 1 1>}}
  User.ColorSwitcher {type=NiagaraFloat, defaultValue=0, description=""}
  User.IsR2 {type=NiagaraBool}
  User.RibbonIDOffset {type=NiagaraInt32}
  User.RibbonWidth {type=NiagaraFloat, defaultValue=10, description=""}
  User.SizeNormalized {type=NiagaraFloat, defaultValue=0.5, description=""}
  User.SwingSectionIndex {type=NiagaraInt32}
 ]
}
